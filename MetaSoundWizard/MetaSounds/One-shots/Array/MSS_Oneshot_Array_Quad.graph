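# A quad one-shot MetaSound Source that plays a single shuffled wave from a Sounds array each time it is played.
# On Play advances a wave asset array Shuffle node, which hands the chosen wave to a quad Wave Player and triggers it.
# The player feeds front and side left/right outputs, and its On Finished trigger ends the one-shot.
# The Sounds array is empty by default.
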MetaSoundSource /MetaSoundWizard/MetaSounds/One-shots/Array/MSS_Oneshot_Array_Quad
inputs [
 Sounds {type=WaveAsset:Array}
 UE.Source.OnPlay {type=Trigger}
]
interfaces [UE.OutputFormat.Quad:1.0, UE.Source.OneShot:1.0, UE.Source:1.1]
outputs [
 "UE.OutputFormat.Quad.Audio:0" {type=Audio}
 "UE.OutputFormat.Quad.Audio:1" {type=Audio}
 "UE.OutputFormat.Quad.Audio:4" {type=Audio}
 "UE.OutputFormat.Quad.Audio:5" {type=Audio}
 UE.Source.OneShot.OnFinished {type=Trigger}
]
pages [Default {
  input_defaults {UE.Source.OnPlay=null, Sounds []}
  nodes [
   n1 = Array.Shuffle:WaveAsset:Array(In Array=^Sounds, Next=^UE.Source.OnPlay) {also {defaults {Enable Shared State="<MetaSoundBuilderResult.SUCCEEDED: 0>"}}}
   WavePlayer_0 = UE.Wave Player:Quad(Play=@n1.On Next, Wave Asset=@n1.Value)
  ]
  links [
   {dst="@UE.Source.OneShot.OnFinished", src=WavePlayer_0.On Finished}
   {dst="@UE.OutputFormat.Quad.Audio:0", src=WavePlayer_0.Out Front Left}
   {dst="@UE.OutputFormat.Quad.Audio:1", src=WavePlayer_0.Out Front Right}
   {dst="@UE.OutputFormat.Quad.Audio:4", src=WavePlayer_0.Out Side Left}
   {dst="@UE.OutputFormat.Quad.Audio:5", src=WavePlayer_0.Out Side Right}
  ]
 }]
properties {assetImportData=&self:AssetImportData, numChannels=4, outputFormat=Quad}
